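; Season 17 quest ability that tracks damage the player deals without being damaged in return.
; Damage-dealt events add to a running total when the target is a player pawn or a bot pawn with qualifying NPC or hostile tags.
; A parallel damage-received listener resets the total to zero.
; Reaching DamageTarget sends the DamageWithoutBeingDamaged stat event to the owner's quest manager and ends the ability.
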(event Ability|EventActivateAbility
  (|CallParentFunction)
  (bind (_asynctaskproxy _payload) (Ability|Tasks|WaitGameplayEvent "(TagName=\"Event.Damage.Dealt\")" 0 false false)
    (:then
      (Ability|Tasks|WaitGameplayEvent "(TagName=\"Event.Damage.Received\")"
        (:EventReceived
          (Variables|Default|SetTotalDamage 0.0))))
    (:EventReceived
      (bind (_payload_eventtag _payload_instigator _payload_target _payload_optionalobject _payload_optionalobject2 _payload_contexthandle _payload_instigatortags _payload_targettags _payload_eventmagnitude) (Utilities|Struct|BreakGameplayEventData _payload))
      (Utilities|Casting|CastToFortPlayerPawn _payload_target
        (:then
          _)
        (:CastFailed
          (Utilities|Casting|CastToFortBotPawn _payload_target)
          (if (GameplayTags|HasAnyTags _payload_targettags (GameplayTags|MakeGameplayTagContainerfromArray (Utilities|Array|MakeArray "(TagName=\"Homebase.Class\")" "(TagName=\"Athena.Faction.Generic.Hostile\")" "(TagName=\"Pawn.Athena.NPC.Duel.IsBeingChallenged\")" "(TagName=\"Pawn.Athena.NPC.Mang\")" "(TagName=\"Pawn.Athena.NPC.IO\")" "(TagName=\"AISpawnerData.Type.MotherWall\")")))
            _)))
      (bind _output_get (Variables|Default|SetTotalDamage (+ (Variables|Default|GetTotalDamage) _payload_eventmagnitude)))
      (if (>= _output_get (Variables|Default|GetDamageTarget))
        (Quest|SendComplexCustomStatEvent (Variables|GAVars|Data|GetOwningPlayerQuestManager) 0 "(GameplayTags=((TagName=\"Athena.Quest.DamageWithoutBeingDamaged\")))" "(GameplayTags=)" 1)
        (Ability|EndAbility)))))

(event Ability|EventFailedToActivatePassiveAbility
  (params (ActorInfo :type Gameplay%20Ability%20Actor%20Info%20Structure)))

(event Ability|EventOnEndAbility
  (params (bWasCancelled :type Boolean)))
